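# MetaSound Source preset that supplies the interactive music for the Puppeteer boss, built on the standard NPC music template.
# The parent graph holds the logic; this preset fills in percussion and non-percussion loops for three phases, transition and death stingers, tempo, mix levels and fades.
# Exposed inputs cover NPC phase events, combat and health state, and beat-quantization divisions.
# Output is stereo, using template attenuation and play-when-silent virtualization.
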MetaSoundSource /PuppeteerBoss/Audio/Music/MSS_NPCMusic_Puppeteer
preset_parent=/FortNPCMusic/Audio/MSS_NPCMusic_Template_Standard1
inputs [
 ActiveSound.bIsOccluded {type=Bool}
 ActiveSound.MaxDistance {type=Float}
 Debug {type=Bool}
 Mixing.DrumMult.Phase0 {type=Float}
 Mixing.DrumMult.Phase1 {type=Float}
 Mixing.DrumMult.Phase2 {type=Float}
 Mixing.Phase0 {type=Float}
 Mixing.Phase1 {type=Float}
 Mixing.Phase2 {type=Float}
 Mixing.Stinger.Activate {type=Float}
 Mixing.Stinger.Death {type=Float}
 Mixing.Stinger.Phase0 {type=Float}
 Mixing.Stinger.Phase1 {type=Float}
 Mixing.Stinger.Phase2 {type=Float}
 NPC.Event.OnMusicStop {type=Trigger}
 NPC.Event.OnPhase0 {type=Trigger}
 NPC.Event.OnPhase1 {type=Trigger}
 NPC.Event.OnPhase2 {type=Trigger}
 NPC.Event.OnPlayerLost {type=Trigger}
 NPC.Event.OnPlayerSpotted {type=Trigger}
 NPC.Music.bEnableOcclusion {type=Bool}
 NPC.Music.Metronome {type=Float}
 NPC.Music.OnEnterCombatPhase {type=Float}
 NPC.Music.StartingPhase {type=Float}
 NPC.Music.VolumeRadius {type=Float}
 NPC.State.bInCombat {type=Bool}
 NPC.State.bListenForTargeting {type=Bool}
 NPC.State.HealthPct {type=Float}
 NPC.State.HealthThreshold {type=Float}
 Phase0.FadeInTime {type=Time}
 Phase0.FadeOutTime {type=Time}
 Phase1.FadeInTime {type=Time}
 Phase1.FadeOutTime {type=Time}
 Phase2.FadeInTime {type=Time}
 Phase2.FadeOutTime {type=Time}
 Quantize.Phase0.LoopBeatDivision {type=Int32}
 Quantize.Phase0.StingerBeatDivision {type=Int32}
 Quantize.Phase1.LoopBeatDivision {type=Int32}
 Quantize.Phase1.StingerBeatDivision {type=Int32}
 Quantize.Phase2.LoopBeatDivision {type=Int32}
 Quantize.Phase2.StingerBeatDivision {type=Int32}
 Quantize.StingerDeath.StingerBeatDivision {type=Int32}
 UE.Attenuation.Distance {type=Float}
 UE.Source.OnPlay {type=Trigger}
 UE.Spatialization.Azimuth {type=Float}
 UE.Spatialization.Elevation {type=Float}
 Wave.Phase0.DrumLoop {type=WaveAsset}
 Wave.Phase0.ElseLoop {type=WaveAsset}
 Wave.Phase0.ExtLoop {type=WaveAsset}
 Wave.Phase1.DrumLoop {type=WaveAsset}
 Wave.Phase1.ElseLoop {type=WaveAsset}
 Wave.Phase1.ExtLoop {type=WaveAsset}
 Wave.Phase2.DrumLoop {type=WaveAsset}
 Wave.Phase2.ElseLoop {type=WaveAsset}
 Wave.Phase2.ExtLoop {type=WaveAsset}
 Wave.Stinger.Death {type=WaveAsset}
 Wave.Stinger.Transition.Phase0 {type=WaveAsset}
 Wave.Stinger.Transition.Phase1 {type=WaveAsset}
 Wave.Stinger.Transition.Phase2 {type=WaveAsset}
 Wave.Stinger.Transition.Phase3 {type=WaveAsset}
]
interfaces [Fort.ActiveSound:1.0, UE.Attenuation:1.0, UE.OutputFormat.Stereo:1.0, UE.Source.OneShot:1.0, UE.Source:1.1, UE.Spatialization:1.0]
outputs [
 CurrentPhase {type=Int32}
 "UE.OutputFormat.Stereo.Audio:0" {type=Audio}
 "UE.OutputFormat.Stereo.Audio:1" {type=Audio}
 UE.Source.OneShot.OnFinished {type=Trigger}
]
pages [Default {input_defaults {
   ActiveSound.MaxDistance=1
   ActiveSound.bIsOccluded=null
   Debug=null
   Mixing.DrumMult.Phase0=1
   Mixing.DrumMult.Phase1=1
   Mixing.DrumMult.Phase2=1
   Mixing.Phase0=1.1
   Mixing.Phase1=1.2
   Mixing.Phase2=1.2
   Mixing.Stinger.Activate=1
   Mixing.Stinger.Death=1
   Mixing.Stinger.Phase0=0.2
   Mixing.Stinger.Phase1=1
   Mixing.Stinger.Phase2=0.5
   NPC.Event.OnMusicStop=null
   NPC.Event.OnPhase0=null
   NPC.Event.OnPhase1=null
   NPC.Event.OnPhase2=null
   NPC.Event.OnPlayerLost=null
   NPC.Event.OnPlayerSpotted=null
   NPC.Music.Metronome=126
   NPC.Music.OnEnterCombatPhase=1
   NPC.Music.StartingPhase=0
   NPC.Music.VolumeRadius=12840.001
   NPC.Music.bEnableOcclusion=null
   NPC.State.HealthPct=0
   NPC.State.HealthThreshold=0
   NPC.State.bInCombat=null
   NPC.State.bListenForTargeting="<MetaSoundBuilderResult.SUCCEEDED: 0>"
   Phase0.FadeInTime=0.1
   Phase0.FadeOutTime=1
   Phase1.FadeInTime=0.2
   Phase1.FadeOutTime=1
   Phase2.FadeInTime=0.2
   Phase2.FadeOutTime=1
   Quantize.Phase0.LoopBeatDivision=1
   Quantize.Phase0.StingerBeatDivision=1
   Quantize.Phase1.LoopBeatDivision=1
   Quantize.Phase1.StingerBeatDivision=1
   Quantize.Phase2.LoopBeatDivision=1
   Quantize.Phase2.StingerBeatDivision=1
   Quantize.StingerDeath.StingerBeatDivision=1
   UE.Attenuation.Distance=0
   UE.Source.OnPlay=null
   UE.Spatialization.Azimuth=0
   UE.Spatialization.Elevation=0
   Wave.Phase0.DrumLoop=/PuppeteerBoss/Audio/Music/wav/MX_PuppeteerBoss_Phase_0_Loop_Perc
   Wave.Phase0.ElseLoop=/PuppeteerBoss/Audio/Music/wav/MX_PuppeteerBoss_Phase_0_Loop_Else
   Wave.Phase0.ExtLoop=null
   Wave.Phase1.DrumLoop=/PuppeteerBoss/Audio/Music/wav/MX_PuppeteerBoss_Phase_1_Loop_Perc
   Wave.Phase1.ElseLoop=/PuppeteerBoss/Audio/Music/wav/MX_PuppeteerBoss_Phase_1_Loop_Else
   Wave.Phase1.ExtLoop=null
   Wave.Phase2.DrumLoop=/PuppeteerBoss/Audio/Music/wav/MX_PuppeteerBoss_Phase_2_Loop_Perc
   Wave.Phase2.ElseLoop=/PuppeteerBoss/Audio/Music/wav/MX_PuppeteerBoss_Phase_2_Loop_Else
   Wave.Phase2.ExtLoop=null
   Wave.Stinger.Death=/PuppeteerBoss/Audio/Music/wav/MX_PuppeteerBoss_Death_Stinger
   Wave.Stinger.Transition.Phase0=/PuppeteerBoss/Audio/Music/wav/MX_PuppeteerBoss_Sus_Intro_Stinger
   Wave.Stinger.Transition.Phase1=/PuppeteerBoss/Audio/Music/wav/MX_PuppeteerBoss_Phase_0_to_1_Stinger
   Wave.Stinger.Transition.Phase2=/PuppeteerBoss/Audio/Music/wav/MX_PuppeteerBoss_Phase_1_to_2_Stinger
   Wave.Stinger.Transition.Phase3=null
  }}]
properties {
 assetImportData=&self:AssetImportData
 attenuationSettings=/FortNPCMusic/Audio/AS_NPC_Music_Template
 maxDistance=16000
 numChannels=2
 outputFormat=Stereo
 sampleRate=48000
 soundClassObject=&Fort_Music_NPC
 virtualizationMode=PlayWhenSilent
}
