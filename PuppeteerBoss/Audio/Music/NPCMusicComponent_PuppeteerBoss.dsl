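; Music component for the Puppeteer boss that changes the music according to the boss's combat stage.
; On BeginPlay it runs the parent logic, then casts its owner to BP_PlayerPawn_Puppeteer_Demon and stores it as BossRef.
; The UpdateMusic event switches on ePuppeteerCombatStage and passes itself and the matching Phase 0, 1 or 2 trigger name to the music manager's ExecuteMusicTrigger.
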(event EventBeginPlay
  (|CallParentFunction)
  (bind _asbp_player_pawn_puppeteer_demon (Utilities|Casting|CastToBP_PlayerPawn_Puppeteer_Demon (Components|GetOwner)))
  (Variables|Default|SetBossRef _asbp_player_pawn_puppeteer_demon))

(event EventTick
  (params (DeltaSeconds :type Float%20%28single-precision%29)))

(event Custom|UpdateMusic
  (params (CurrentBattleStage :type ePuppeteerCombatStage%20Enum :default "NewEnumerator0"))
  (switch Utilities|FlowControl|Switch|SwitchonePuppeteerCombatStage CurrentBattleStage
    (:NewEnumerator0
      (Class|NPCMusicManagerComponent|ExecuteMusicTrigger (Variables|Private|GetMusicManager) self (Variables|Default|GetPhase0_TriggerName)))
    (:NewEnumerator1
      (Class|NPCMusicManagerComponent|ExecuteMusicTrigger (Variables|Private|GetMusicManager) self (Variables|Default|GetPhase1_TriggerName)))
    (:NewEnumerator2
      (Class|NPCMusicManagerComponent|ExecuteMusicTrigger (Variables|Private|GetMusicManager) self (Variables|Default|GetPhase2_TriggerName)))))
